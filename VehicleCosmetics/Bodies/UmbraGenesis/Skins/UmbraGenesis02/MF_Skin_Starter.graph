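# Skin material function for the UmbraGenesis02 vehicle body, outputting material attributes with an animated fire look that responds to speed.
# Base color and emissive layer color parameters, character art and noise-distorted texture masks, with a switchable UV-based symmetry mask.
# Two panning, Perlin-warped flame samples alternate over time and scale with the Speedometer parameter, which SpeedDebug replaces with a sine.
# Several branches, including an Octane-texture pattern, end unconnected.
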MaterialFunction /VehicleCosmetics/Bodies/UmbraGenesis/Skins/UmbraGenesis02/MF_Skin_Starter
graph {nodes [
  n1 = NamedRerouteUsage()
  n2 = ComponentMask(None=@n1) {r=true}
  n3 = ConstantBiasScale(None=@n2) {bias=-0.6, scale=4}
  n4 = MaterialFunctionCall!(
   Texture Coordinates
   Tiling X=Constant() {r=8}
   Tiling Y=Constant() {r=5}
   PanSpeedX=Constant() {r=0.1}
   PanSpeedY=Constant() {r=0.5}
   Angle
   RotateSpeed
   Balance
   Chaos
  ) {materialFunction=/VehicleCosmetics/Materials/MaterialFunctions/MF_Noisemaker_Perlin}
  n5 = OneMinus(None=@n4.Perlin1)
  Add(A=@n3, B=@n5)
  n6 = ComponentMask(None=TextureCoordinate() {coordinateIndex=1}) {g=true}
  n7 = Constant() {r=32}
  n8 = MaterialFunctionCall!(
   Texture Coordinates
   Tiling X=@n7
   Tiling Y=@n7
   Offset=Constant()
   Angle
   Sharpen=Constant() {r=1}
   PanSpeedX
   PanSpeedY=Constant() {r=0.5}
   RotateSpeed
  ) {materialFunction=/VehicleCosmetics/Materials/MaterialFunctions/MF_Noisemaker_Dots}
  n9 = Constant() {r=0.1}
  n10 = Saturate(None=Multiply(A=Multiply(A=@n8.Dots, B=@n8.Dots), B=@n9))
  n11 = Multiply(A=Subtract(A=@n8.Dots, B=@n8.Dots), B=@n9) {constB=0.3}
  ConstantBiasScale(None=Subtract(A=Add(A=@n6, B=@n10), B=Saturate(None=@n11))) {bias=-0.3, scale=20}
  Constant()
  Constant() {desc=size, r=3}
  Constant() {r=1}
  n12 = TextureSampleParameter2D() {parameterName=RGB2, texture=&/T_UmbraGenesis02_RGBA2}
  n13 = NamedRerouteUsage()
  n14 = ScalarParameter() {defaultValue=1.5, parameterName=EmissiveIntensity}
  n15 = SetMaterialAttributes!(
   MaterialAttributes
   BaseColor=LinearInterpolate(A=NamedRerouteUsage(), B=Constant3Vector() {constant=<rgba 0.0677 0.0677 0.0677 0>}, Alpha=@n12.A)
   Metallic=LinearInterpolate(Alpha=@n12.A)
   Roughness=LinearInterpolate(Alpha=@n12.A) {constA=0.25, constB=0.5}
   EmissiveColor=LinearInterpolate(A=Multiply(A=@n13, B=@n14), Alpha=@n12.A) {constB=0}
   Normal
   AmbientOcclusion
  ) {attributeSetTypes [
    #1
    #2
    #3
    #4
    #5
    #6
   ]}
  FunctionOutput(None=@n15)
  Max(A=@n3, B=ConstantBiasScale(None=Multiply(A=@n2, B=@n5)) {bias=-0.1, scale=2})
  n16 = TextureCoordinate() {coordinateIndex=1}
  n17 = ComponentMask(None=@n16) {r=true}
  n18 = LinearInterpolate(A=@n17, B=OneMinus(None=@n17), Alpha=ScalarParameter() {parameterName=InvertRedChannelSymmetryMask})
  n19 = Ceil(None=Subtract(A=@n18) {constB=0.65})
  n20 = Ceil(None=Subtract(A=ComponentMask(None=@n16) {g=true}) {constB=0.4})
  Max(A=@n19, B=@n20)
  Multiply(A=NamedRerouteUsage()) {constB=0.5}
  n21 = TextureSampleParameter2D() {parameterName=RGB2, texture=&T_UmbraGenesis02_RGBA2}
  Multiply(A=@n21.R) {constB=0.1}
  NamedRerouteDeclaration(None=Constant()) {name=MetalMask, nodeColor=<rgba 0 1 0.4141 1>}
  NamedRerouteDeclaration(None=VectorParameter() {parameterName=Color1, defaultValue=<rgba 0.8963 0 0.0018 1>}.RGB) {name=TeamColor, nodeColor=<rgba 0.003 0.0179 1 1>}
  NamedRerouteDeclaration(None=VectorParameter() {parameterName=Color2, defaultValue=<rgba 0 0 0 1>}.RGB) {name=CustomColor, nodeColor=<rgba 0 1 0.0625 1>}
  NamedRerouteDeclaration(None=MaterialFunctionCall!(Fresnel Power=Constant() {r=3}, Normal=NamedRerouteUsage()) {materialFunction=/VehicleCosmetics/Materials/MaterialFunctions/MF_Fresnel_Prime}.Result) {name=Fresnel, nodeColor=<rgba 0.7344 0 1 1>}
  NamedRerouteDeclaration(None=Time() {bOverride_Period=true, period=500}) {name=time, nodeColor=<rgba 1 0.9285 0 1>}
  n22 = NamedRerouteUsage()
  n23 = NamedRerouteUsage()
  n24 = LinearInterpolate(A=@n22, B=@n23, Alpha=NamedRerouteUsage()) {constB=0}
  n25 = TextureSample() {texture=&/T_UmbraGenesis01_RGB_BG}
  n26 = NamedRerouteUsage()
  n27 = TextureSampleParameter2D() {parameterName=RGB2, texture=&T_UmbraGenesis02_RGBA2}
  n28 = Multiply(A=Add(A=Multiply(A=@n26) {constB=0.25}, B=@n27.R) {constB=0}) {constB=6}
  n29 = Subtract(A=@n28, B=Floor(None=@n28))
  n30 = Ceil(None=Subtract(A=@n18) {constB=0.5})
  n31 = Saturate(None=Multiply(A=ConstantBiasScale(None=@n27.R) {bias=-0.1, scale=500}, B=OneMinus(None=@n30)) {constB=4})
  n32 = Abs(None=ConstantBiasScale(None=Multiply(A=@n29, B=@n31) {constB=4}) {bias=-0.5, scale=1})
  n33 = LinearInterpolate(
   A=Multiply(A=@n22) {constB=0.15}
   B=LinearInterpolate(A=@n23, B=LinearInterpolate(A=@n24, Alpha=@n25.G), Alpha=@n25.R)
   Alpha=Saturate(None=ConstantBiasScale(None=@n32) {bias=-0.2, scale=10})
  ) {constB=0}
  n34 = Clamp(None=@n21.R) {maxDefault=0.1}
  n35 = AppendVector(A=Constant(), B=@n4.Perlin1)
  n36 = TextureSampleParameter2D(UVs=Add(A=TextureCoordinate(), B=Multiply(A=@n35) {constB=-0.01})) {parameterName=RGB2, texture=&T_UmbraGenesis02_RGBA2}
  n37 = Multiply(A=@n5) {constB=0.5}
  n38 = Multiply(A=@n36.R, B=ConstantBiasScale(None=Subtract(A=@n36.R, B=@n37)) {bias=-0.4, scale=10})
  n39 = LinearInterpolate(B=Saturate(None=Max(A=@n34, B=@n38)), Alpha=Saturate(None=Min(A=@n19, B=@n20))) {constB=0}
  n40 = TextureSampleParameter2D() {parameterName=Texture_CharacterMain, texture=&/T_UmbraGenesis02_Chars}
  n41 = AppendVector(
   A=Constant()
   B=MaterialFunctionCall(Tiling X=Constant() {r=10}, Tiling Y=Constant() {r=15}, PanSpeedX=Constant() {r=1}, PanSpeedY=Constant() {r=2}) {materialFunction=&MF_Noisemaker_Perlin}.Perlin1
  )
  n42 = TextureSampleParameter2D(UVs=Add(A=TextureCoordinate(), B=Multiply(A=@n41) {constB=-0.005})) {parameterName=RGB2, texture=&T_UmbraGenesis02_RGBA2}
  n43 = ComponentMask(None=NamedRerouteUsage()) {r=true}
  n44 = ConstantBiasScale(None=@n43) {bias=-0.5, scale=500}
  n45 = Saturate(None=Multiply(A=@n42.B, B=Saturate(None=@n44)))
  n46 = LinearInterpolate(A=LinearInterpolate(A=LinearInterpolate(A=@n33, Alpha=@n39) {constB=0}, B=@n40.RGB, Alpha=@n40.RGB), Alpha=@n45) {constB=0}
  n47 = MaterialFunctionCall(PanSpeedX=Constant() {r=0.5}, PanSpeedY=Constant() {r=1}, Chaos=Constant() {r=2}) {materialFunction=&MF_Noisemaker_Perlin}
  n48 = Saturate(None=Subtract(
    A=TextureSampleParameter2D() {parameterName=RGB2, texture=&T_UmbraGenesis02_RGBA2}.G
    B=Multiply(A=@n47.Perlin1) {constB=0.4}
   ))
  n49 = TextureCoordinate() {coordinateIndex=1}
  n50 = Multiply(A=@n47.Perlin1) {constB=0.05}
  n51 = TextureSample(UVs=Subtract(A=Panner(Coordinate=@n49) {speedY=0.1}, B=@n50)) {texture=&/T_UmbraGenesis02_Flames}
  n52 = TextureCoordinate() {coordinateIndex=1, uTiling=2}
  n53 = TextureSample(UVs=Add(A=Panner(Coordinate=@n52) {speedY=0.05}, B=@n50)) {texture=&T_UmbraGenesis02_Flames}
  n54 = StaticSwitchParameter(
   True=Sine(None=Time() {bOverride_Period=true, period=500}) {period=5}
   False=ScalarParameter() {desc="Speedometer - uses percentage\r\nof speed from VCID", parameterName=Speedometer}
  ) {parameterName=SpeedDebug}
  n55 = NamedRerouteDeclaration(None=Clamp(None=ConstantBiasScale(None=@n54) {bias=-0.1, desc=Balance copied from Pixie, scale=1.2}) {minDefault=0.0001}) {name=Speed, nodeColor=<rgba 1 0 0.4922 1>}
  n56 = Subtract(
   A=Multiply(A=Max(A=@n51.A, B=@n53.A), B=@n55)
   B=TextureSampleParameter2D() {parameterName=Texture_CharacterMain, texture=&T_UmbraGenesis02_Chars}.A
  )
  NamedRerouteDeclaration(None=LinearInterpolate(A=LinearInterpolate(A=@n46, Alpha=@n48) {constB=0}, Alpha=Saturate(None=@n56)) {constB=0}) {name=BaseColor, nodeColor=<rgba 1 0.0938 0 1>}
  NamedRerouteDeclaration(None=Constant3Vector() {constant=<rgba 0 0 1 0>}) {name=NormalTexture, nodeColor=<rgba 0.299 0.3454 1 1>}
  n57 = VectorParameter() {parameterName=FireColor2, defaultValue=<rgba 0.0072 0 1 1>}
  n58 = LinearInterpolate(A=@n57.RGB, B=VectorParameter() {parameterName=FireColor1, defaultValue=<rgba 0 0.5826 0.9823 1>}.RGB, Alpha=Saturate(None=@n3))
  n59 = Multiply(A=@n40.RGB) {constB=0.25}
  n60 = ConstantBiasScale(None=OneMinus(None=ComponentMask(None=@n1) {g=true})) {bias=-0.9, scale=2}
  n61 = ConstantBiasScale(None=@n2) {bias=-0.7, scale=2}
  n62 = Multiply(A=Saturate(None=Max(A=@n60, B=@n61)), B=@n57.RGB)
  n63 = ConstantBiasScale(None=@n43) {bias=-0.5, scale=4}
  n64 = LinearInterpolate(
   A=LinearInterpolate(A=LinearInterpolate(B=@n58, Alpha=@n39) {constB=0}, B=Add(A=@n59, B=@n62), Alpha=@n40.A) {constB=0}
   B=@n58
   Alpha=Saturate(None=Multiply(A=@n45, B=@n63))
  ) {constB=0}
  n65 = Sine(None=NamedRerouteUsage()) {period=2}
  n66 = ConstantBiasScale(None=@n65) {bias=1.2}
  n67 = ConstantBiasScale(None=Multiply(A=@n65) {constB=-1}) {bias=1.2}
  n68 = Max(A=Multiply(A=@n51.RGB, B=Saturate(None=@n66)), B=Multiply(A=@n53.RGB, B=Saturate(None=@n67)))
  n69 = ComponentMask(None=@n68) {g=true}
  n70 = Add(A=@n55, B=Multiply(A=Multiply(A=@n69) {constB=2}, B=@n55) {constB=2})
  n71 = Multiply(A=Saturate(None=ConstantBiasScale(None=@n70) {bias=-1, scale=1})) {constB=3}
  NamedRerouteDeclaration(None=Add(A=LinearInterpolate(A=@n64, Alpha=@n48), B=Multiply(A=@n68, B=@n71))) {name=Emissive, nodeColor=<rgba 0 0.2266 1 1>}
  NamedRerouteDeclaration(None=Constant() {r=0.1}) {name=RoughnessMask, nodeColor=<rgba 0.6641 0 1 1>}
  NamedRerouteUsage()
  NamedRerouteUsage()
  n72 = OneMinus(None=MaterialFunctionCall!(
    Texture Coordinates
    Tiling=Constant() {r=1}
    PanSpeed
    Angle=Constant() {r=90}
    RotateSpeed
    Sharpen=Constant() {r=1}
    Offset=Constant() {r=-0.031}
   ) {materialFunction=/VehicleCosmetics/Materials/MaterialFunctions/MF_Noisemaker_Gradient}.Gradient3)
  n73 = ConstantBiasScale(None=@n72) {bias=-1.95, scale=500}
  n74 = ConstantBiasScale(None=@n72) {bias=-1.955, scale=500}
  n75 = Max(A=Subtract(A=Saturate(None=@n73), B=Saturate(None=@n74)), B=Saturate(None=ConstantBiasScale(None=@n72) {bias=-1.97, scale=500}))
  n76 = TextureCoordinate() {coordinateIndex=1}
  n77 = Abs(None=ConstantBiasScale(None=ComponentMask(None=@n76) {r=true}) {bias=-0.5, scale=2})
  n78 = Multiply(B=ComponentMask(None=@n76) {g=true}) {constA=2}
  n79 = Rotator(
   Coordinate=NamedRerouteDeclaration(None=AppendVector(A=@n77, B=@n78)) {name=symmetricalUVs, nodeColor=<rgba 1 0 0.5625 1>}
   Time=Constant() {r=11}
  )
  n80 = Saturate(None=ConstantBiasScale(None=ComponentMask(None=@n79) {g=true}) {bias=0.25, scale=1})
  n81 = TextureCoordinate() {coordinateIndex=1}
  n82 = Multiply(A=@n81, B=Constant2Vector() {g=1.5, r=1}) {constB=3}
  n83 = Constant2Vector() {g=-0.2, r=-0.05}
  n84 = TextureSampleParameter2D(UVs=Rotator(Coordinate=Add(A=@n82, B=@n83), Time=Constant()) {centerY=0.55, speed=-1.5}) {parameterName=RGB, texture=/VehicleCosmetics/Bodies/Octane/Skins/Octane_UmbraGenesisLTE01/T_UmbraGenesis_RGB1}
  n85 = Add(A=Multiply(A=@n81, B=Constant() {desc=size, r=3}) {constB=3}, B=Constant2Vector() {g=-1.25, r=-0.7})
  n86 = TextureSampleParameter2D(UVs=Rotator(Coordinate=@n85, Time=Constant()) {centerY=0.55, speed=-1.5}) {parameterName=RGB, texture=&T_UmbraGenesis_RGB1}
  n87 = Multiply(A=@n85)
  n88 = Constant2Vector() {g=0.2}
  n89 = TextureSampleParameter2D(UVs=Rotator(Coordinate=Add(A=@n87, B=@n88), Time=Constant()) {speed=1}) {parameterName=RGB, texture=&T_UmbraGenesis_RGB1}
  n90 = ConstantBiasScale(None=@n84.B) {bias=-0.4, scale=50}
  n91 = Max(A=@n84.B, B=LinearInterpolate(A=Max(A=@n86.B, B=@n89.B), Alpha=Saturate(None=@n90)) {constB=0})
  n92 = Multiply(A=Add(A=Constant(), B=@n91)) {constB=1.5}
  n93 = Subtract(A=@n92, B=Floor(None=@n92))
  n94 = Max(A=Saturate(None=ConstantBiasScale(None=@n93) {bias=-0.65, scale=1}), B=Saturate(None=ConstantBiasScale(None=@n91) {bias=-0.98, scale=50}))
  n95 = Saturate(None=ConstantBiasScale(None=@n91) {bias=-0.5, scale=5})
  n96 = ConstantBiasScale(None=@n91) {bias=-0.3, scale=50}
  Saturate(None=LinearInterpolate(A=Max(A=@n75, B=@n80), B=Multiply(A=@n94, B=@n95), Alpha=Saturate(None=@n96)) {constB=0})
  n97 = Constant() {r=1}
  n98 = Constant() {r=1.2}
  n99 = Constant() {r=1}
  n100 = Add(
   A=OneMinus(None=MaterialFunctionCall(Tiling=@n97, Sharpen=@n98, Offset=@n99) {materialFunction=&MF_Noisemaker_Gradient}.Gradient1)
   B=@n91
  )
  n101 = ConstantBiasScale(None=@n100) {bias=-0.6, scale=600}
  n102 = ConstantBiasScale(None=@n100) {bias=-0.7, scale=600}
  n103 = Multiply(A=@n91, B=Subtract(A=Saturate(None=@n101), B=Saturate(None=@n102)))
  n104 = Saturate(None=ConstantBiasScale(None=@n72) {bias=-1.2, scale=1})
  Saturate(None=ConstantBiasScale(None=Multiply(A=@n103, B=@n104)) {bias=-0.01, scale=600})
  Subtract(A=@n42.B, B=Multiply() {constB=0.3})
 ]}
